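# Arena scoring HUD widget for Showdown, a FortHUDPointCounter subclass that shows the player's score beside a shadowed Hype icon on a league-styled panel.
# The inherited score and debug ID texts are laid out here, with the debug text and the division badge overlay collapsed by default.
# PreConstruct binds to the lazy division image's loading-state change and sets its dynamic material's OutlineAmount to 1.
# The OnDivisionImageLoaded event sets the same outline again, ignoring the loaded flag.
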WidgetBlueprint /FortniteCoreUI/Frontend/Showdown/ShowdownScoringHUD_Progress_Arena
graphs [EventGraph {nodes [
   OnDivisionImageLoaded!() -> (OutputDelegate, then, Loaded:bool=~) {is=title K2Node_CustomEvent}
   > n1 = GetDynamicMaterial(self=Get Image_ArenaDivision() -> (Image_ArenaDivision:Fort Lazy Image Object Reference))
   > SetScalarParameterValue(self=@n1, ParameterName=OutlineAmount, Value=1)
   Event PreConstruct()
   > Bind Event to On Loading State Changed(self:Common Lazy Image Object Reference=Get Image_ArenaDivision(), Delegate=Create Event()) {is=title K2Node_AddDelegate}
   > n2 = GetDynamicMaterial(self=Get Image_ArenaDivision())
   > SetScalarParameterValue(self=@n2, ParameterName=OutlineAmount, Value=1)
  ]}]
variables []
widgets {
 info {inheritedWidgetCount=2, parentClass=/Script/FortniteUI.FortHUDPointCounter, rootWidgetClass=&Overlay, widgetCount=15}
 widgets [
  Overlay_0 {widgetClassPath=&Overlay, widget=Overlay_0, properties {visibility=HitTestInvisible}}
  Overlay_1 {widgetClassPath=&Overlay, parent=Overlay_0, widget=Overlay_1, slot {class=OverlaySlot}}
  SizeBox_0 {
   widgetClassPath=&SizeBox
   parent=Overlay_1
   widget=SizeBox_0
   properties {bOverride_HeightOverride=true, heightOverride=46}
   slot {class=OverlaySlot, properties {horizontalAlignment=HAlign_Fill, verticalAlignment=VAlign_Fill}}
  }
  Overlay_3 {widgetClassPath=&Overlay, parent=SizeBox_0, widget=Overlay_3, slot {class=SizeBoxSlot}}
  Image_PanelBG {
   widgetClassPath=&Image
   bIsVariable=true
   parent=Overlay_3
   widget=Image_PanelBG
   properties {
    visibility=SelfHitTestInvisible
    brush {
     bIsDynamicallyLoaded=false
     drawAs=Box
     imageType=NoImage
     mirroring=NoMirror
     resourceName=None
     resourceObject=/FortniteCoreUI/Frontend/Showdown/Materials/league/M_LeaguePanelBG_Inst
     tiling=NoTile
     imageSize=<100 100>
     margin=<ltrb 0 0.17 0 0.17>
     outlineSettings {
      bUseBrushTransparency=false
      roundingType=HalfHeightRadius
      width=0
      color {colorUseRule=UseColor_Specified, specifiedColor=<rgba 0 0 0 0>}
      cornerRadii=<0 0 0 0>
     }
     tintColor {colorUseRule=UseColor_Specified, specifiedColor=<rgba 1 1 1 1>}
     uVRegion {bIsValid=false, max=<0 0>, min=<0 0>}
    }
   }
   slot {class=OverlaySlot, properties {horizontalAlignment=HAlign_Fill, verticalAlignment=VAlign_Fill}}
  }
  Image_PanelBorder {
   widgetClassPath=&Image
   bIsVariable=true
   parent=Overlay_3
   widget=Image_PanelBorder
   properties {
    visibility=SelfHitTestInvisible
    brush {
     bIsDynamicallyLoaded=false
     drawAs=Box
     imageType=NoImage
     mirroring=NoMirror
     resourceName=None
     resourceObject=/Game/UI/Frontend/Showdown/Materials/M_UI_SkewBoxBorder_TournamentPanelBorder
     tiling=NoTile
     imageSize=<100 100>
     margin=<ltrb 0 0.17 0 0.17>
     outlineSettings {
      bUseBrushTransparency=false
      roundingType=HalfHeightRadius
      width=0
      color {colorUseRule=UseColor_Specified, specifiedColor=<rgba 0 0 0 0>}
      cornerRadii=<0 0 0 0>
     }
     tintColor {colorUseRule=UseColor_Specified, specifiedColor=<rgba 1 1 1 1>}
     uVRegion {bIsValid=false, max=<0 0>, min=<0 0>}
    }
   }
   slot {class=OverlaySlot, properties {horizontalAlignment=HAlign_Fill, verticalAlignment=VAlign_Fill}}
  }
  HorizontalBox_0 {
   widgetClassPath=&HorizontalBox
   parent=Overlay_1
   widget=HorizontalBox_0
   slot {class=OverlaySlot, properties {verticalAlignment=VAlign_Center, padding=<ltrb 10 0 10 0>}}
  }
  Overlay_5 {
   widgetClassPath=&Overlay
   parent=HorizontalBox_0
   widget=Overlay_5
   slot {class=HorizontalBoxSlot, properties {horizontalAlignment=HAlign_Left, verticalAlignment=VAlign_Center, padding=<ltrb 5 0 0 0>}}
  }
  Image_65 {
   widgetClassPath=&Image
   bIsVariable=true
   parent=Overlay_5
   widget=Image_65
   properties {
    brush {
     bIsDynamicallyLoaded=false
     drawAs=Image
     imageType=NoImage
     mirroring=NoMirror
     resourceName=None
     resourceObject=/FortniteCoreUI/Frontend/Showdown/Materials/league/M_UI_LeagueDivision_PlayerRatingIcon
     tiling=NoTile
     imageSize=<25 25>
     margin=<ltrb 0 0 0 0>
     outlineSettings {
      bUseBrushTransparency=false
      roundingType=HalfHeightRadius
      width=0
      color {colorUseRule=UseColor_Specified, specifiedColor=<rgba 0 0 0 0>}
      cornerRadii=<0 0 0 0>
     }
     tintColor {colorUseRule=UseColor_Specified, specifiedColor=<rgba 1 1 1 1>}
     uVRegion {bIsValid=false, max=<0 0>, min=<0 0>}
    }
    colorAndOpacity=<rgba 0.0052 0.0052 0.0056 0.4026>
   }
   slot {class=OverlaySlot}
  }
  HypeIcon {
   widgetClassPath=&Image
   bIsVariable=true
   parent=Overlay_5
   widget=HypeIcon
   properties {
    brush {
     bIsDynamicallyLoaded=false
     drawAs=Image
     imageType=NoImage
     mirroring=NoMirror
     resourceName=None
     resourceObject=&M_UI_LeagueDivision_PlayerRatingIcon
     tiling=NoTile
     imageSize=<25 25>
     margin=<ltrb 0 0 0 0>
     outlineSettings {
      bUseBrushTransparency=false
      roundingType=HalfHeightRadius
      width=0
      color {colorUseRule=UseColor_Specified, specifiedColor=<rgba 0 0 0 0>}
      cornerRadii=<0 0 0 0>
     }
     tintColor {colorUseRule=UseColor_Specified, specifiedColor=<rgba 1 1 1 1>}
     uVRegion {bIsValid=false, max=<0 0>, min=<0 0>}
    }
    renderTransform {angle=0, scale=<1 1>, shear=<0 0>, translation=<-4 -3>}
   }
   slot {class=OverlaySlot}
  }
  Text_ScoreValue {
   widgetClassPath=&CommonTextBlock
   bInherited=true
   bIsVariable=true
   parent=HorizontalBox_0
   widget=Text_ScoreValue
   properties {
    text="1,000,000"
    font {
     bForceMonospaced=false
     bMaterialIsStencil=false
     fontMaterial=null
     fontObject=&Burbank
     letterSpacing=0
     monospacedWidth=1
     size=24
     skewAmount=0
     typefaceFontName=Burbank-PlayerPower
     outlineSettings {
      bApplyOutlineToDropShadows=false
      bMiteredCorners=false
      bSeparateFillAlpha=false
      outlineMaterial=null
      outlineSize=1
      outlineColor=<rgba 0.0052 0.0052 0.0056 0.258>
     }
    }
    shadowColorAndOpacity=<rgba 0.0052 0.0052 0.0056 0.4026>
    shadowOffset=<4 3>
   }
   slot {class=HorizontalBoxSlot, properties {horizontalAlignment=HAlign_Left, verticalAlignment=VAlign_Center, padding=<ltrb 0 -15 0 -15>}}
  }
  Overlay_4 {
   widgetClassPath=&Overlay
   parent=Overlay_1
   widget=Overlay_4
   properties {visibility=Collapsed, renderTransform {angle=0, scale=<1 1>, shear=<0 0>, translation=<-29 0>}}
   slot {class=OverlaySlot}
  }
  Image_64 {
   widgetClassPath=&Image
   bIsVariable=true
   parent=Overlay_4
   widget=Image_64
   properties {
    visibility=SelfHitTestInvisible
    brush {
     bIsDynamicallyLoaded=false
     drawAs=Image
     imageType=NoImage
     mirroring=NoMirror
     resourceName=None
     resourceObject=/FortniteCoreUI/Frontend/Showdown/Materials/M_UI_Arena_HUDBadgeBG
     tiling=NoTile
     imageSize=<60 60>
     margin=<ltrb 0 0 0 0>
     outlineSettings {
      bUseBrushTransparency=false
      roundingType=HalfHeightRadius
      width=0
      color {colorUseRule=UseColor_Specified, specifiedColor=<rgba 0 0 0 0>}
      cornerRadii=<0 0 0 0>
     }
     tintColor {colorUseRule=UseColor_Specified, specifiedColor=<rgba 1 1 1 1>}
     uVRegion {bIsValid=false, max=<0 0>, min=<0 0>}
    }
    renderTransform {angle=45, scale=<1 1>, shear=<0 0>, translation=<0 3>}
   }
   slot {class=OverlaySlot, properties {horizontalAlignment=HAlign_Center, verticalAlignment=VAlign_Center}}
  }
  Image_ArenaDivision {
   widgetClassPath=&FortLazyImage
   bIsVariable=true
   parent=Overlay_4
   widget=Image_ArenaDivision
   properties {
    bIsImageForDesignerPreviewOnly=true
    brush {
     bIsDynamicallyLoaded=false
     drawAs=Image
     imageType=NoImage
     mirroring=NoMirror
     resourceName=None
     resourceObject=/FortniteCoreUI/Frontend/Showdown/Materials/league/M_UI_LeagueDivision_Icon6
     tiling=NoTile
     imageSize=<64 64>
     margin=<ltrb 0 0 0 0>
     outlineSettings {
      bUseBrushTransparency=false
      roundingType=HalfHeightRadius
      width=0
      color {colorUseRule=UseColor_Specified, specifiedColor=<rgba 0 0 0 0>}
      cornerRadii=<0 0 0 0>
     }
     tintColor {colorUseRule=UseColor_Specified, specifiedColor=<rgba 1 1 1 1>}
     uVRegion {bIsValid=false, max=<0 0>, min=<0 0>}
    }
   }
   slot {class=OverlaySlot, properties {horizontalAlignment=HAlign_Center, verticalAlignment=VAlign_Center}}
  }
  Text_DebugId {
   widgetClassPath=&TextBlock
   bInherited=true
   parent=Overlay_0
   widget=Text_DebugId
   properties {
    autoWrapText=true
    text=EventId: Event EventWindowId: EventWindow
    visibility=Collapsed
    wrapTextAt=150
    wrappingPolicy=AllowPerCharacterWrapping
    colorAndOpacity {colorUseRule=UseColor_Specified, specifiedColor=<rgba 0 0 0 1>}
    font {
     bForceMonospaced=false
     bMaterialIsStencil=false
     fontMaterial=null
     fontObject=&Roboto
     letterSpacing=0
     monospacedWidth=1
     size=12
     skewAmount=0
     typefaceFontName=Bold
     outlineSettings {
      bApplyOutlineToDropShadows=false
      bMiteredCorners=false
      bSeparateFillAlpha=false
      outlineMaterial=null
      outlineSize=1
      outlineColor=<rgba 1 1 1 1>
     }
    }
   }
   slot {class=OverlaySlot}
  }
 ]
}
