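# Arena HUD scoring widget, built on FortHUDScoringAlert, that pairs a point-scored popup with an embedded progress point counter.
# OnPointScored shows the blurred popup, plays ScoreAnimation and a sound cue.
# When the animation finishes, it collapses the popup and calls HandlePointAnimationFinished.
# PreConstruct positions the counter for mobile or non-mobile; Construct asynchronously awaits playlist data, shifting it right for a visible tactical overshield in non-touch, non-mobile NoBuildBR playlists.
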WidgetBlueprint /FortniteCoreUI/UI/Competitive/Arena/ArenaScoringHUD
animations [ScoreAnimation {
  binding_count=4
  display_label=""
  length_seconds=5
  object=&self:ScoreAnimation
  sequence {bindings [
    Text_ScoreDesc {
     id=#1
     possessed_class=CommonTextBlock
     tracks [{
       type=MovieScene2DTransformTrack
       sections [{
         type=MovieScene2DTransformSection
         end_frame=100
         start_frame=0
         channels [
          Translation.X {type=MovieSceneScriptingFloatChannel, keys [[0,-54.5397],[21,0]]}
          Translation.Y {type=MovieSceneScriptingFloatChannel}
          Angle {type=MovieSceneScriptingFloatChannel}
          Scale.X {type=MovieSceneScriptingFloatChannel}
          Scale.Y {type=MovieSceneScriptingFloatChannel}
          Shear.X {type=MovieSceneScriptingFloatChannel}
          Shear.Y {type=MovieSceneScriptingFloatChannel}
         ]
        }]
      }]
    }
    Overlay_0 {
     id=#2
     display_name=CommonBorder_PointsContainer
     possessed_class=Overlay
     tracks [{
       type=MovieSceneFloatTrack
       display_name=Render Opacity
       sections [{
         type=MovieSceneFloatSection
         end_frame=100
         start_frame=0
         channels [None {type=MovieSceneScriptingFloatChannel, keys [[0,0],[15.1393,1],[94.7893,1],[100,0]]}]
        }]
      }]
    }
    HorizontalBox {
     id=#3
     display_name=ScoringContent
     possessed_class=HorizontalBox
     tracks [
      {
       type=MovieSceneByteTrack
       display_name=Visibility
       sections [{type=MovieSceneByteSection, end_frame=100, start_frame=0, channels [None {type=MovieSceneScriptingByteChannel, keys [[0,3],[100,1]]}]}]
      }
      {
       type=MovieSceneColorTrack
       display_name=Content Color and Opacity
       sections [{
         type=MovieSceneColorSection
         end_frame=100
         start_frame=0
         channels [
          Color.R {type=MovieSceneScriptingFloatChannel, const [1,1]}
          Color.G {type=MovieSceneScriptingFloatChannel, const [1,1]}
          Color.B {type=MovieSceneScriptingFloatChannel, const [1,1]}
          Color.A {type=MovieSceneScriptingFloatChannel, keys [[0,0],[10,1],[94.7893,1],[100,0]]}
         ]
        }]
      }
      {
       type=MovieScene2DTransformTrack
       sections [{
         type=MovieScene2DTransformSection
         end_frame=100
         start_frame=0
         channels [
          Translation.X {type=MovieSceneScriptingFloatChannel, keys [[0,-282.8563],[10,0],[94.7893,0],[100,-100]]}
          Translation.Y {type=MovieSceneScriptingFloatChannel}
          Angle {type=MovieSceneScriptingFloatChannel}
          Scale.X {type=MovieSceneScriptingFloatChannel}
          Scale.Y {type=MovieSceneScriptingFloatChannel}
          Shear.X {type=MovieSceneScriptingFloatChannel}
          Shear.Y {type=MovieSceneScriptingFloatChannel}
         ]
        }]
      }
     ]
    }
    Text_PointCount {
     id=#4
     possessed_class=CommonTextBlock
     tracks [
      {
       type=MovieScene2DTransformTrack
       sections [{
         type=MovieScene2DTransformSection
         end_frame=100
         start_frame=0
         channels [
          Translation.X {type=MovieSceneScriptingFloatChannel}
          Translation.Y {type=MovieSceneScriptingFloatChannel}
          Angle {type=MovieSceneScriptingFloatChannel}
          Scale.X {type=MovieSceneScriptingFloatChannel, keys [[15.1393,1],[16,1.1982],[19,0.9804],[19.8487,1]]}
          Scale.Y {type=MovieSceneScriptingFloatChannel, keys [[15.1393,1],[16,1.1982],[19,0.9804],[19.8487,1]]}
          Shear.X {type=MovieSceneScriptingFloatChannel}
          Shear.Y {type=MovieSceneScriptingFloatChannel}
         ]
        }]
      }
      {
       type=MovieSceneWidgetMaterialTrack
       display_name=Font.FontMaterial
       sections [{
         type=MovieSceneParameterSection
         channels [TimeOffset {type=MovieSceneScriptingFloatChannel, keys [[16.5183,0.8425],[30.2603,-1.3558],[37.5823,0.8425],[51.3243,-1.3558]]}]
        }]
      }
     ]
    }
   ]}
 }]
graphs [
 AdjustWidgetForOvershield {nodes [
   AdjustWidgetForOvershield() {is=title K2Node_FunctionEntry}
   > n1 = IsTacticalOvershieldEnabled!(execute) -> (then, ReturnValue:bool)
   n2 = GetPlayerStateZone!(self:Fort HUDContext Object Reference=Get FortHUDContext!() -> (ReturnValue:Fort HUDContext Object Reference) {is=title K2Node_GetBlueprintContextEx}) -> (ReturnValue:Fort Player State Zone Object Reference)
   > Branch(Condition=AND Boolean(A=@n1, B=ShouldOvershieldBarBeVisible!(self:Fort Player State Zone Object Reference=@n2) -> (ReturnValue:bool)))
   n3 = Get PointCounterContent() -> (PointCounterContent:Overlay Object Reference)
   n4 = BreakVector2D(InVec=Break Widget Transform!(WidgetTransform:"Widget Transform Structure (by ref)"=Get RenderTransform!(self:Widget Object Reference=@n3) -> (RenderTransform:Widget Transform Structure)) -> (Translation:Vector 2D Structure, Scale:Vector 2D Structure, Shear:Vector 2D Structure, Angle:float=~) {is=title K2Node_BreakStruct})
   > SetRenderTranslation(self=@n3, Translation=MakeVector2D(X=float + float(A=@n4, B=40), Y=@n4.Y))
  ]}
 EventGraph {nodes [
   n1 = ScoreFinished() {is=title K2Node_CustomEvent}
   > SetVisibility(self=Get CommonBorder_PointsContainer() -> (CommonBorder_PointsContainer:Common Border Object Reference), InVisibility=Collapsed)
   > HandlePointAnimationFinished(self:Fort HUDScoring Alert Object Reference)
   Event Construct()
   > BindToAnimationFinished!(
    execute
    self:Widget Animation Object Reference=Get ScoreAnimation() -> (ScoreAnimation:Widget Animation Object Reference)
    Widget:User Widget Object Reference=Self-Reference()
    Delegate=@n1
   ) -> (then)
   > n2 = PlaylistDataReadyAsync!(execute:Exec) -> (
    then:Exec
    OnReady:Exec
    GameState:Fort Game State Athena Object Reference
    Playlist:Fort Playlist Object Reference
    PlaylistContextTags:"Gameplay Tag Container Structure (by ref)"
   ) {is=title K2Node_AsyncAction}
   n3 = OR Boolean(A=IsMobileGame(), B=IsUsingTouch(PlayerController=GetOwningPlayer()))
   Branch(
    execute=@n2.OnReady
    Condition=AND Boolean(
     A=HasTag!(
      TagContainer:"Gameplay Tag Container Structure (by ref)"=@n2.PlaylistContextTags
      Tag:Gameplay Tag Structure=(TagName="Athena.Playlist.NoBuildBR")
      bExactMatch:bool=false
     ) -> (ReturnValue:bool)
     B=NOT Boolean(A=@n3)
    )
   )
   > AdjustWidgetForOvershield() {is=title K2Node_CallFunction}
   Event PreConstruct()
   > n4 = Branch(Condition=IsMobileGame())
   > SetRenderTransform!(
    execute
    self:Widget Object Reference=Get PointCounterContent()
    InTransform:Widget Transform Structure=Make Widget Transform!(Translation:Vector 2D Structure=(X=400,Y=-225), Scale:Vector 2D Structure=(X=1.1,Y=1.1), Shear:Vector 2D Structure=(X=0,Y=0), Angle:float=0) -> (WidgetTransform:Widget Transform Structure) {is=title K2Node_MakeStruct}
   ) -> (then)
   SetRenderTransform(
    execute=@n4.else
    self=Get PointCounterContent()
    InTransform=Make Widget Transform(Translation=(X=630,Y=-95), Scale=(X=1,Y=1), Shear=(X=0,Y=0), Angle=0)
   )
   Event OnPointScored() {is=title K2Node_Event}
   > PlayAnimation(InAnimation=Get ScoreAnimation())
   > SetVisibility(self=Get CommonBorder_PointsContainer(), InVisibility=HitTestInvisible)
   > PlaySound2D(Sound=/Game/Sounds/Fort_UI_Sounds/FrontEnd/Store/Fort_UI_Showdown_Point_Cue)
  ]}
]
variables [ScoringRuleDisplayIcons {
  type="array<Texture2D>"
  default [
   /Game/UI/Foundation/Textures/Icons/Quest/T-Icon-Trophy-32
   /Game/Athena/HUD/Art/killsico_alt.KillsIco_Alt
   /FortniteCoreUI/UI/Events/Icons/T-Icon-Bus-Fare-Flat
  ]
 }]
widgets {
 info {inheritedWidgetCount=4, parentClass=/Script/FortniteUI.FortHUDScoringAlert, rootWidgetClass=&Overlay, widgetCount=18}
 widgets [
  Overlay_0 {widgetClassPath=&Overlay, widget=Overlay_0}
  CommonBorder_PointsContainer {
   widgetClassPath=&CommonBorder
   bIsVariable=true
   parent=Overlay_0
   widget=CommonBorder_PointsContainer
   properties {visibility=Collapsed, padding=<ltrb 0 0 0 0>}
   slot {class=OverlaySlot, properties {verticalAlignment=VAlign_Center}}
  }
  BackgroundBlurPoints {
   widgetClassPath=&BackgroundBlur
   parent=CommonBorder_PointsContainer
   widget=BackgroundBlurPoints
   properties {blurStrength=40}
   slot {class=BorderSlot, properties {padding=<ltrb 0 0 0 0>}}
  }
  Overlay_2 {
   widgetClassPath=&Overlay
   parent=BackgroundBlurPoints
   widget=Overlay_2
   slot {class=BackgroundBlurSlot, properties {padding=<ltrb 0 0 0 0>}}
  }
  Image_39 {
   widgetClassPath=&Image
   parent=Overlay_2
   widget=Image_39
   properties {brush {
     bIsDynamicallyLoaded=false
     drawAs=Box
     imageType=NoImage
     mirroring=NoMirror
     resourceName=None
     resourceObject=/FortniteCoreUI/Frontend/Showdown/Materials/league/M_LeaguePanelBG_Inst
     tiling=NoTile
     imageSize=<32 36.9007>
     margin=<ltrb 0 0 0 0>
     outlineSettings {
      bUseBrushTransparency=false
      roundingType=HalfHeightRadius
      width=0
      color {colorUseRule=UseColor_Specified, specifiedColor=<rgba 0 0 0 0>}
      cornerRadii=<0 0 0 0>
     }
     tintColor {colorUseRule=UseColor_Specified, specifiedColor=<rgba 1 1 1 1>}
     uVRegion {bIsValid=false, max=<0 0>, min=<0 0>}
    }}
   slot {class=OverlaySlot, properties {horizontalAlignment=HAlign_Fill, verticalAlignment=VAlign_Bottom}}
  }
  ScoringContent {
   widgetClassPath=&CommonBorder
   parent=Overlay_2
   widget=ScoringContent
   properties {verticalAlignment=VAlign_Center, visibility=Collapsed, padding=<ltrb 4 2 10 2>}
   slot {class=OverlaySlot}
  }
  Overlay_4 {
   widgetClassPath=&Overlay
   parent=ScoringContent
   widget=Overlay_4
   slot {class=BorderSlot, properties {verticalAlignment=VAlign_Center, padding=<ltrb 4 2 10 2>}}
  }
  VerticalBox_0 {
   widgetClassPath=&VerticalBox
   parent=Overlay_4
   widget=VerticalBox_0
   slot {class=OverlaySlot, properties {padding=<ltrb 80 0 0 0>}}
  }
  Text_ScoreDesc {
   widgetClassPath=&CommonTextBlock
   bInherited=true
   bIsVariable=true
   parent=VerticalBox_0
   widget=Text_ScoreDesc
   properties {
    justification=Right
    text=4 ELIMINATIONS
    textTransformPolicy=ToUpper
    font {
     bForceMonospaced=false
     bMaterialIsStencil=false
     fontMaterial=null
     fontObject=&Burbank
     letterSpacing=0
     monospacedWidth=1
     size=18
     skewAmount=0
     typefaceFontName=Burbank-PlayerPower
     outlineSettings {
      bApplyOutlineToDropShadows=false
      bMiteredCorners=false
      bSeparateFillAlpha=false
      outlineMaterial=null
      outlineSize=0
      outlineColor=<rgba 0 0 0 1>
     }
    }
   }
   slot {
    class=VerticalBoxSlot
    properties {horizontalAlignment=HAlign_Right, verticalAlignment=VAlign_Center, padding=<ltrb 0 2 0 0>, size {sizeRule=Fill, value=1}}
   }
  }
  Text_PointCount {
   widgetClassPath=&CommonTextBlock
   bInherited=true
   bIsVariable=true
   parent=VerticalBox_0
   widget=Text_PointCount
   properties {
    justification=Right
    text=+1 POINT
    font {
     bForceMonospaced=false
     bMaterialIsStencil=false
     fontMaterial=null
     fontObject=&Burbank
     letterSpacing=0
     monospacedWidth=1
     size=24
     skewAmount=0
     typefaceFontName=Burbank-Black
     outlineSettings {
      bApplyOutlineToDropShadows=false
      bMiteredCorners=false
      bSeparateFillAlpha=false
      outlineMaterial=null
      outlineSize=1
      outlineColor=<rgba 0.0052 0.0052 0.0056 1>
     }
    }
    renderTransform {angle=0, scale=<1 1>, shear=<-7 0>, translation=<0 0>}
    renderTransformPivot=<0.9 0.8>
    shadowColorAndOpacity=<rgba 0.0052 0.0052 0.0056 1>
    shadowOffset=<3 3>
   }
   slot {class=VerticalBoxSlot, properties {horizontalAlignment=HAlign_Right, verticalAlignment=VAlign_Center, size {sizeRule=Fill, value=1}}}
  }
  Overlay_1 {
   widgetClassPath=&Overlay
   parent=Overlay_4
   widget=Overlay_1
   slot {class=OverlaySlot, properties {verticalAlignment=VAlign_Center}}
  }
  Overlay_3 {
   widgetClassPath=&Overlay
   parent=Overlay_1
   widget=Overlay_3
   properties {renderTransform {angle=0, scale=<1 1>, shear=<0 0>, translation=<0 2>}}
   slot {class=OverlaySlot, properties {padding=<ltrb 0 -3 0 0>}}
  }
  Image_ScoreShadow {
   widgetClassPath=&Image
   parent=Overlay_3
   widget=Image_ScoreShadow
   properties {
    brush {
     bIsDynamicallyLoaded=false
     drawAs=Image
     imageType=NoImage
     mirroring=NoMirror
     resourceName=None
     resourceObject=/FortniteCoreUI/Frontend/Showdown/Materials/M_UI_Arena_HUDBadgeBG
     tiling=NoTile
     imageSize=<60 60>
     margin=<ltrb 0 0 0 0>
     outlineSettings {
      bUseBrushTransparency=false
      roundingType=HalfHeightRadius
      width=0
      color {colorUseRule=UseColor_Specified, specifiedColor=<rgba 0 0 0 0>}
      cornerRadii=<0 0 0 0>
     }
     tintColor {colorUseRule=UseColor_Specified, specifiedColor=<rgba 1 1 1 1>}
     uVRegion {bIsValid=false, max=<0 0>, min=<0 0>}
    }
    colorAndOpacity=<rgba 0.0052 0.0052 0.0056 0.45>
    renderTransform {angle=45, scale=<1 1>, shear=<0 0>, translation=<2 4>}
   }
   slot {class=OverlaySlot, properties {horizontalAlignment=HAlign_Center, verticalAlignment=VAlign_Center}}
  }
  Image_ScoreButton {
   widgetClassPath=&Image
   parent=Overlay_3
   widget=Image_ScoreButton
   properties {
    brush {
     bIsDynamicallyLoaded=false
     drawAs=Image
     imageType=NoImage
     mirroring=NoMirror
     resourceName=None
     resourceObject=&M_UI_Arena_HUDBadgeBG
     tiling=NoTile
     imageSize=<60 60>
     margin=<ltrb 0 0 0 0>
     outlineSettings {
      bUseBrushTransparency=false
      roundingType=HalfHeightRadius
      width=0
      color {colorUseRule=UseColor_Specified, specifiedColor=<rgba 0 0 0 0>}
      cornerRadii=<0 0 0 0>
     }
     tintColor {colorUseRule=UseColor_Specified, specifiedColor=<rgba 1 1 1 1>}
     uVRegion {bIsValid=false, max=<0 0>, min=<0 0>}
    }
    renderTransform {angle=45, scale=<1 1>, shear=<0 0>, translation=<0 0>}
   }
   slot {class=OverlaySlot, properties {horizontalAlignment=HAlign_Center, verticalAlignment=VAlign_Center}}
  }
  Icon_Scoring {
   widgetClassPath=/Script/CommonUI.CommonLazyImage
   bInherited=true
   bIsVariable=true
   parent=Overlay_1
   widget=Icon_Scoring
   properties {brush {
     bIsDynamicallyLoaded=false
     drawAs=Image
     imageType=NoImage
     mirroring=NoMirror
     resourceName=None
     resourceObject=&KillsIco_Alt
     tiling=NoTile
     imageSize=<32 32>
     margin=<ltrb 0 0 0 0>
     outlineSettings {
      bUseBrushTransparency=false
      roundingType=HalfHeightRadius
      width=0
      color {colorUseRule=UseColor_Specified, specifiedColor=<rgba 0 0 0 0>}
      cornerRadii=<0 0 0 0>
     }
     tintColor {colorUseRule=UseColor_Specified, specifiedColor=<rgba 1 1 1 1>}
     uVRegion {bIsValid=false, max=<0 0>, min=<0 0>}
    }}
   slot {class=OverlaySlot, properties {horizontalAlignment=HAlign_Center, verticalAlignment=VAlign_Center}}
  }
  PointCounterContent {
   widgetClassPath=&Overlay
   bIsVariable=true
   parent=Overlay_0
   widget=PointCounterContent
   properties {renderTransform {angle=0, scale=<1.1 1.1>, shear=<0 0>, translation=<400 -225>}, renderTransformPivot=<1 1>}
   slot {class=OverlaySlot, properties {verticalAlignment=VAlign_Bottom}}
  }
  HUDPointCounter {
   widgetClassPath=/FortniteCoreUI/Frontend/Showdown/ShowdownScoringHUD_Progress_Arena
   bIsVariable=true
   parent=PointCounterContent
   widget=HUDPointCounter
   properties {renderTransformPivot=<1 1>}
   slot {class=OverlaySlot}
  }
  TutorialHighlight {widgetClassPath=&TutorialHighlight}
 ]
}
